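; Debug audio Blueprint that registers a mutator to catch raw SoundWaves played directly, outside Shipping builds.
; On BeginPlay it checks the build configuration and adds its SoundWaveDetectorMutator variable to the Audio Mutators engine subsystem.
; Per the graph's note, the mutator reports offenders via Ensure and mutes them, using an exact-class check to skip MetaSounds and SoundCues.
; Overlap, Tick and the construction script are empty.
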(event EventBeginPlay
  ; Detects raw SoundWaves played directly and reports them via Ensure + mutes (Volume 0).
  ; Exact-class check skips MetaSounds/SoundCues. Disabled in Shipping.
  (if (Utilities|String|NotEqualExactly(String) (Development|GetBuildConfiguration) "Shipping")
    (Audio|Mutator|AddMutator (AudioEngineSubsystems|GetAudioMutatorsSubsystem) (Variables|Default|GetSoundWaveDetectorMutator))))

(event Collision|EventActorBeginOverlap
  (params (OtherActor :type Actor%20Object%20Reference)))

(event EventTick
  (params (DeltaSeconds :type Float%20%28single-precision%29)))

(fn ConstructionScript)
